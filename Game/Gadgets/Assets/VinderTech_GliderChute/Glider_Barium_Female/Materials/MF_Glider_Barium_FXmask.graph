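# Material function for the female Barium glider's animated paint-and-fire effect, outputting emissive color, opacity and board base color.
# Barium_MaskAnimation_curve feeds animation values; noise-distorted sphere masks on pre-skinned position form growing board and logo masks with edges.
# Panning noise, partly in radial coordinates, builds fire masks shaped by the FX mask texture.
# BariumPaintColor tints glow and board over the logo texture; threshold checks switch emissive off and opacity to full.
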MaterialFunction /Game/Gadgets/Assets/VinderTech_GliderChute/Glider_Barium_Female/Materials/MF_Glider_Barium_FXmask
graph {nodes [
  n1 = NamedRerouteUsage()
  n2 = Constant() {r=1}
  n3 = Constant()
  n4 = LinearInterpolate(Alpha=NamedRerouteUsage()) {constA=1, constB=0}
  n5 = TextureSample() {
   samplerSource=SSM_Wrap_WorldGroupSettings
   texture=/Game/Gadgets/Assets/VinderTech_GliderChute/Glider_Barium_Female/Textures/T_Glider_FX_Mask
  }
  n6 = VectorParameter() {group=Glider_Barium, parameterName=BariumPaintColor, defaultValue=<rgba 0.4793 0.6445 0.0578 1>}
  n7 = NamedRerouteUsage()
  n8 = LinearInterpolate(Alpha=@n7) {constA=0.6, constB=0}
  n9 = Multiply(
   A=Power(Base=@n5.B) {constExponent=3}
   B=Multiply(A=Desaturation(None=@n6.RGB, Fraction=@n8), B=LinearInterpolate(Alpha=@n7) {constA=0.5}) {constB=15}
  )
  n10 = NamedRerouteUsage()
  n11 = Multiply(A=Add(A=Multiply(A=@n5.R, B=@n10), B=NamedRerouteUsage()), B=Multiply(A=@n6.RGB, B=Constant() {r=3}))
  n12 = NamedRerouteUsage()
  n13 = Saturate(None=LinearInterpolate(Alpha=NamedRerouteUsage()) {constB=3})
  n14 = NamedRerouteUsage()
  n15 = Multiply(A=Multiply(A=LinearInterpolate(A=@n12, B=@n10, Alpha=@n13), B=Power(Base=@n14)), B=Multiply(A=@n6.RGB) {constB=3})
  n16 = NamedRerouteUsage()
  n17 = Multiply(A=Add(A=Multiply(A=@n4, B=@n9), B=Add(A=@n11, B=@n15)), B=LinearInterpolate(Alpha=Add(A=@n16) {constB=0.1}) {constA=2, constB=0})
  FunctionOutput(None=NamedRerouteDeclaration(None=If(A=@n1, B=@n2, "A > B"=@n3, "A == B"=@n3, "A < B"=@n17)) {name=Emisssive Color, nodeColor=<rgba 0.1719 1 0 1>}) {outputName=Emisssive Color}
  n18 = NamedRerouteUsage()
  n19 = Constant() {r=1}
  n20 = Constant() {r=1}
  n21 = NamedRerouteUsage()
  n22 = TextureSample() {texture=&T_Glider_FX_Mask}
  n23 = NamedRerouteUsage()
  n24 = Add(A=Saturate(None=Multiply(A=@n21, B=@n22.R)), B=Add(A=Multiply(A=@n23, B=@n22.G), B=@n22.G))
  n25 = NamedRerouteUsage()
  n26 = NamedRerouteUsage()
  n27 = NamedRerouteUsage()
  n28 = Add(A=Multiply(A=Saturate(None=@n24), B=NamedRerouteUsage()), B=Multiply(A=Multiply(A=@n25) {constB=2}, B=Max(A=@n26, B=@n27)))
  FunctionOutput(None=NamedRerouteDeclaration(None=If(A=@n18, B=@n19, "A > B"=@n20, "A == B"=@n20, "A < B"=@n28)) {name=Opacity, nodeColor=<rgba 0 1 0.0156 1>}) {outputName=Opacity}
  n29 = TextureSampleParameter2D() {
   group=Glider_Barium
   parameterName=Barium_LogoTexture
   texture=/Game/Gadgets/Assets/VinderTech_GliderChute/Glider_Barium_Female/Textures/T_Glider_Barium_Female_D2
  }
  n30 = TextureSample() {samplerType=SAMPLERTYPE_Masks, texture=/Game/Gadgets/Assets/VinderTech_GliderChute/Glider_Barium_Female/Textures/T_Glider_Barium_Female_FX}
  FunctionOutput(None=LinearInterpolate(A=@n29.RGB, B=Multiply(A=@n6.RGB), Alpha=@n30.B)) {outputName=Board_basecolor}
  Multiply(A=@n6.RGB, B=@n29.RGB)
  n31 = ComponentMask(None=MaterialFunctionCall(UVs=TextureCoordinate(), "Rotation Angle (0-1)"=Constant() {r=0.9}) {materialFunction=&CustomRotator}.Rotated Values) {g=true, r=true}
  n32 = ComponentMask(None=Multiply(A=Constant3Vector() {constant=<rgba 2 2 0 0>}, B=Constant() {r=1.5}) {constB=4}) {g=true, r=true}
  n33 = Constant()
  n34 = Add(A=Multiply(A=@n31, B=@n32), B=@n33)
  n35 = Multiply(A=Time(), B=Constant() {r=2})
  n36 = Multiply(A=@n35) {constB=0.5}
  n37 = MaterialFunctionCall(
   In=TextureSample(UVs=Panner(Coordinate=@n34, Time=@n36) {speedX=0.05, speedY=-1}) {texture=&T_NoisePack_01}.G
   Contrast=Constant() {r=0.2}
  ) {materialFunction=&CheapContrast}
  n38 = Multiply(A=@n32, B=@n31)
  n39 = TextureSample(UVs=Panner(Coordinate=Add(A=@n38, B=@n33), Time=@n35) {speedX=-0.1, speedY=-1}) {texture=&T_NoisePack_01}
  n40 = Panner(Coordinate=Add(A=@n33, B=Multiply(A=@n32, B=@n31)), Time=@n35) {speedX=0.1, speedY=-1}
  NamedRerouteDeclaration(None=Multiply(A=Add(A=@n37.Result, B=@n39.G), B=TextureSample(UVs=@n40) {texture=&T_NoisePack_01}.G)) {name=firemask_2, nodeColor=<rgba 0.5 0 1 1>}
  n41 = Multiply(A=TextureCoordinate()) {constB=7}
  n42 = PreSkinnedPosition()
  n43 = Add(
   A=Multiply(A=TextureSample(UVs=@n41) {texture=&T_NoisePack_01}.RGB) {constB=10}
   B=ShaderStageSwitch(PixelShader=VertexInterpolator(VS=@n42).PS, VertexShader=@n42)
  )
  n44 = Constant() {r=1}
  n45 = Constant() {r=4}
  n46 = Saturate(None=Multiply(A=NamedRerouteUsage()) {constB=2})
  n47 = AppendVector(A=AppendVector(A=LinearInterpolate(A=@n44, B=@n45, Alpha=@n46), B=Constant() {r=1}), B=Constant() {r=1})
  n48 = NamedRerouteUsage()
  n49 = LinearInterpolate(A=Constant3Vector() {constant=<rgba 5 -15 0 0>}, B=Constant3Vector() {constant=<rgba 0 20 0 0>}, Alpha=@n48)
  n50 = LinearInterpolate(Alpha=@n48) {constB=240}
  n51 = OneMinus(None=SphereMask(A=Multiply(A=@n43, B=@n47), B=@n49, Radius=@n50, Hardness=Constant() {r=0.5}))
  n52 = PreSkinnedPosition()
  n53 = ShaderStageSwitch(PixelShader=VertexInterpolator(VS=@n52).PS, VertexShader=@n52)
  n54 = TextureCoordinate()
  n55 = Multiply(A=TextureSample(UVs=Multiply(A=@n54) {constB=3}) {texture=&T_NoisePack_01}.RGB) {constB=5}
  n56 = SphereMask(A=Multiply(A=@n47, B=Add(A=@n53, B=@n55)), B=@n49, Radius=@n50, Hardness=Constant() {r=1})
  NamedRerouteDeclaration(None=Power(Base=Multiply(A=@n51, B=@n56)) {constExponent=3}) {name=logo_mask, nodeColor=<rgba 1 0 0.8906 1>}
  NamedRerouteDeclaration(None=@n56) {name=logo_inter_mask_, nodeColor=<rgba 1 0 0.8906 1>}
  n57 = Constant3Vector() {constant=<rgba 6 1 1 0>}
  n58 = PreSkinnedPosition()
  n59 = VertexInterpolator(VS=@n58)
  n60 = TextureSample(UVs=Multiply(A=TextureCoordinate()) {constB=3}) {texture=&T_NoisePack_01}
  n61 = Multiply(A=@n57, B=Add(A=ShaderStageSwitch(PixelShader=@n59.PS, VertexShader=@n58), B=Multiply(A=@n60.RGB) {constB=5}))
  n62 = Saturate(None=Subtract(A=NamedRerouteUsage()) {constB=0.06})
  n63 = LinearInterpolate(A=Constant3Vector() {constant=<rgba 20 -15 0 0>}, B=Constant3Vector() {constant=<rgba 0 20 0 0>}, Alpha=@n62)
  n64 = LinearInterpolate(Alpha=@n62) {constB=240}
  n65 = SphereMask(A=@n61, B=@n63, Radius=@n64, Hardness=Constant() {r=1})
  NamedRerouteDeclaration(None=@n65) {name=board_inter_mask__1, nodeColor=<rgba 1 0 0.8906 1>}
  n66 = ComponentMask(None=MaterialFunctionCall() {materialFunction=&BoundingBoxBased_0-1_UVW}.RGB) {g=true, r=true}
  n67 = Constant2Vector() {r=-2}
  n68 = MaterialFunctionCall(Vector or UVs=MaterialFunctionCall(UVs=Add(A=@n66, B=@n67), Texture Scale=Constant2Vector() {g=1, r=4}) {materialFunction=&ScaleUVsByCenter}.UVs) {materialFunction=&VectorToRadialValue}
  n69 = ComponentMask(None=Constant() {r=2}) {g=true, r=true}
  n70 = Multiply(A=@n68.Radial Coordinates, B=@n69)
  n71 = Constant()
  n72 = Multiply(A=Time(), B=Constant() {r=1})
  n73 = TextureSample(UVs=Panner(Coordinate=Add(A=@n70, B=@n71), Time=Multiply(A=@n72) {constB=0.5}) {speedX=0.05, speedY=-1}) {texture=&T_NoisePack_01}
  n74 = Constant() {r=0.2}
  n75 = Panner(Coordinate=Add(A=Multiply(A=@n69, B=@n68.Radial Coordinates), B=@n71), Time=@n72) {speedX=-0.1, speedY=-1}
  n76 = Add(A=@n71, B=Multiply(A=@n69, B=@n68.Radial Coordinates))
  n77 = Multiply(
   A=Add(
    A=MaterialFunctionCall(In=@n73.G, Contrast=@n74) {materialFunction=&CheapContrast}.Result
    B=TextureSample(UVs=@n75) {texture=&T_NoisePack_01}.G
   )
   B=TextureSample(UVs=Panner(Coordinate=@n76, Time=@n72) {speedX=0.1, speedY=-1}) {texture=&T_NoisePack_01}.G
  )
  n78 = Add(A=Multiply(A=NamedRerouteUsage(), B=Saturate(None=@n77))) {constB=0.01}
  n79 = NamedRerouteUsage()
  n80 = Constant() {r=0.03}
  n81 = TextureSample(UVs=Add(A=TextureCoordinate(), B=Multiply(A=@n79, B=@n80))) {
   samplerSource=SSM_Wrap_WorldGroupSettings
   texture=&T_Glider_FX_Mask
  }
  n82 = Subtract(A=Add(A=Power(Base=@n81.A) {constExponent=1.5}) {constB=0.05}, B=@n5.G)
  n83 = Subtract(A=NamedRerouteUsage()) {constB=0.5}
  n84 = Constant()
  n85 = Subtract(A=NamedRerouteUsage()) {constB=0.2}
  n86 = Multiply(A=If(A=@n83, B=Constant(), "A > B"=@n83, "A == B"=@n84, "A < B"=@n84), B=LinearInterpolate(Alpha=Saturate(None=@n85)))
  n87 = LinearInterpolate(Alpha=Saturate(None=Multiply(A=@n82, B=@n86))) {constB=50}
  n88 = TextureSample() {
   samplerSource=SSM_Wrap_WorldGroupSettings
   texture=&T_Glider_FX_Mask
  }
  NamedRerouteDeclaration(None=Multiply(A=Multiply(A=@n78, B=@n87) {constB=8}, B=OneMinus(None=@n88.R))) {name=firemask_stage2, nodeColor=<rgba 0.5 0 1 1>}
  n89 = Panner(Coordinate=Multiply(A=TextureCoordinate()) {constB=7}) {speedX=0.5, speedY=-0.05}
  n90 = PreSkinnedPosition()
  n91 = Add(
   A=Multiply(A=TextureSample(UVs=@n89) {texture=&T_NoisePack_01}.RGB) {constB=10}
   B=ShaderStageSwitch(PixelShader=VertexInterpolator(VS=@n90).PS, VertexShader=@n90)
  )
  n92 = OneMinus(None=SphereMask(A=Multiply(A=@n91, B=@n57), B=@n63, Radius=@n64, Hardness=Constant() {r=0.5}))
  NamedRerouteDeclaration(None=Power(Base=Multiply(A=@n92, B=@n65)) {constExponent=3}) {name=board_mask_1, nodeColor=<rgba 1 0 0.8906 1>}
  n93 = Subtract(A=@n64) {constB=0.06}
  n94 = Constant() {r=0.95}
  NamedRerouteDeclaration(None=Subtract(A=@n65, B=SphereMask(A=@n61, B=@n63, Radius=@n93, Hardness=@n94))) {name=board_mask_edge_1, nodeColor=<rgba 1 0 0.8906 1>}
  n95 = ScalarParameter() {defaultValue=0.4, group=Glider_Barium, parameterName=Barium_MaskAnimation_curve, sliderMax=1}
  n96 = LinearInterpolate(Alpha=NamedRerouteDeclaration(None=OneMinus(None=@n95)) {name=Main_MaskAnimation, nodeColor=<rgba 1 0.9141 0 1>}) {constB=7}
  n97 = Subtract(A=@n96, B=Constant() {r=1.4})
  n98 = Constant() {r=5}
  NamedRerouteDeclaration(None=If(A=@n96, B=Constant() {r=1.7}, "A > B"=Multiply(A=@n97, B=@n98), "A == B"=@n96, "A < B"=@n96)) {name=Animation, nodeColor=<rgba 0.1016 0 1 1>}
 ]}
