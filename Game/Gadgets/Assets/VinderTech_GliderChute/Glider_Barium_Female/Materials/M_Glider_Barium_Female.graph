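# Masked, two-sided master material for the Barium Female glider, supporting skeletal, cloth, morph and Nanite meshes with per-pixel shading model selection.
# The shared character PreFX function feeds CharacterSkinFX.
# A glider-specific FX mask function then overrides attributes such as emissive, opacity mask and base color on that result.
# The modified attributes pass through the shared character PostFX and cinematic PostFX functions.
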Material /Game/Gadgets/Assets/VinderTech_GliderChute/Glider_Barium_Female/Materials/M_Glider_Barium_Female
graph {
 properties {
  bAutomaticallySetUsageInEditor=false
  bUseMaterialAttributes=true
  bUsedWithClothing=true
  bUsedWithMeshDeformer=true
  bUsedWithMorphTargets=true
  bUsedWithNanite=true
  bUsedWithSkeletalMesh=true
  blendMode=BLEND_Masked
  pixelDepthOffsetMode=PDOM_Legacy
  refractionMethod=RM_IndexOfRefraction
  shadingModel=MSM_FromMaterialExpression
  subsurfaceProfile=&SS_HeroSkin_02
  twoSided=true
  usedShadingModels=Default Lit | Subsurface | Clear Coat | Subsurface Profile | Cloth | Eye
 }
 nodes [
  n1 = MaterialFunctionCall() {materialFunction=&BaseFN_CharacterFunction_PreFX}
  n2 = MaterialFunctionCall(
   PreskinnedPosition=@n1.Pre skinned Local Position
   Gmap=@n1.Pre skinned Local Position
   MaterialAttributes=@n1.Pre skinned Local Position
   AO=@n1.Pre skinned Local Position
  ) {materialFunction=&CharacterSkinFX}
  n3 = MaterialFunctionCall() {materialFunction=&/MF_Glider_Barium_FXmask}
  n4 = MaterialFunctionCall(
   AO=@n1.AO
   Distance Blend=@n1.AO
   Fresnel=@n1.AO
   Gmap=@n1.AO
   Pre Skinned Local Position=@n1.AO
   Material=SetMaterialAttributes!(
    MaterialAttributes=GetMaterialAttributes(None=@n2.Result).MaterialAttributes
    Emissive Color=@n3.Emisssive Color
    Opacity Mask=@n3.Emisssive Color
    Base Color=@n3.Emisssive Color
   ) {attributeSetTypes [#1, #2, #3]}
  ) {materialFunction=&BaseFN_CharacterFunction_PostFX}
  MaterialFunctionCall(In=@n4.Result) {materialFunction=&CINE_MF_CharPostFX}
 ]
}
